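# Translucent unlit, two-sided material for lightning on Niagara mesh particles, used by the Candor pickaxe FX.
# SplineThicken sets ribbon width, while time-scrolled sine waves along V add local X/Y wobble to world position offset.
# Particle color drives emissive; sine UV masks shape glow or core opacity, cut by step or smoke-texture HeightLerp dissolve.
# Dynamic parameters control width, dissolve, mask choice and wobble; depth fade is skipped on forward and mobile.
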Material /Game/Weapons/FORT_Melee/Pickaxe_Candor/FX/M_Common_Mesh_lightning
graph {
 properties {
  bUsedWithNiagaraMeshParticles=true
  blendMode=BLEND_Translucent
  pixelDepthOffsetMode=PDOM_Legacy
  shadingModel=MSM_Unlit
  twoSided=true
  usedShadingModels=Unlit
 }
 nodes [
  n1 = Multiply(A=TextureCoordinate() {vTiling=2}, B=AppendVector(A=Constant() {r=1}, B=ScalarParameter() {defaultValue=1, parameterName=WPO V Scale}))
  n2 = ScalarParameter() {defaultValue=-3, parameterName=WPO Speed}
  n3 = Add(A=ComponentMask(None=@n1) {g=true}, B=Multiply(A=Time(), B=@n2) {constB=2})
  n4 = Transform(None=Multiply(A=Sine(None=@n3), B=Constant3Vector() {constant=<rgba 1 0 0 0>})) {transformSourceType=TRANSFORMSOURCE_Local}
  n5 = DynamicParameter() {defaultValue=<rgba 2 0 1 10>, paramNames [WidthBase, Dissolve, Use center mask, WPO Intensity]}
  n6 = Add(A=ComponentMask(None=@n1) {g=true}, B=Multiply(A=Time(), B=@n2) {constB=2})
  n7 = Transform(None=Multiply(A=Sine(None=@n6), B=Constant3Vector() {constant=<rgba 0 1 0 0>})) {transformSourceType=TRANSFORMSOURCE_Local}
  n8 = Multiply(
   A=DynamicParameter() {defaultValue=<rgba 2 0 1 10>, paramNames [WidthBase, Dissolve, Use center mask, WPO Intensity]}.WidthBase
   B=ScalarParameter() {defaultValue=1, desc=Allows you to thicken the lightning for glow renderers., parameterName=WIdthScalar, sliderMax=1}
  )
  Add(
   A=MaterialFunctionCall(WidthBase=@n8, WidthTip=@n8) {materialFunction=&SplineThicken}.WorldPositionOffset
   B=Add(A=Multiply(A=@n4, B=@n5.WPO Intensity), B=Multiply(A=@n7, B=@n5.WPO Intensity))
  )
  n9 = ParticleColor()
  MaterialFunctionCall(Emissive=@n9.RGB) {materialFunction=&MF_FX_EmissiveAdjustments}
  n10 = Sine(None=ComponentMask(None=TextureCoordinate()) {r=true}) {period=2}
  n11 = ScalarParameter() {defaultValue=2, parameterName=Glow edges power}
  n12 = TextureCoordinate()
  n13 = Power(Base=Sine(None=ComponentMask(None=@n12) {r=true}) {period=2}, Exp=ScalarParameter() {defaultValue=15, parameterName=Center edges power}) {constExponent=15}
  n14 = LinearInterpolate(
   A=Saturate(None=Power(Base=@n10, Exp=@n11))
   B=Saturate(None=Multiply(A=@n13) {constB=10})
   Alpha=DynamicParameter() {defaultValue=<rgba 2 0 1 10>, paramNames [WidthBase, Dissolve, Use center mask, WPO Intensity]}.Use center mask
  )
  n15 = TextureCoordinate()
  n16 = Multiply(A=Sine(None=ComponentMask(None=@n15) {g=true}) {period=2}) {constB=2}
  n17 = TextureCoordinate()
  n18 = OneMinus(None=Sine(None=ComponentMask(None=@n17) {r=true}) {period=2})
  n19 = DynamicParameter() {defaultValue=<rgba 2 0 1 10>, paramNames [WidthBase, Dissolve, Use center mask, WPO Intensity]}
  n20 = MaterialFunctionCall(
   Transition Phase=DynamicParameter() {defaultValue=<rgba 2 0 1 10>, paramNames [WidthBase, Dissolve, Use center mask, WPO Intensity]}.Dissolve
   Height Texture=TextureSampleParameter2D(UVs=TextureCoordinate() {vTiling=5}) {parameterName=Dissolve Tex, texture=&T_Inky_Smoke_Tile}.R
  ) {materialFunction=&HeightLerp}
  n21 = Multiply(
   A=Multiply(A=Multiply(A=@n9.A, B=@n14), B=Saturate(None=@n16))
   B=LinearInterpolate(A=Step(Y=@n18, X=@n19.Dissolve), B=Multiply(A=@n20.Results), Alpha=ScalarParameter() {defaultValue=1, parameterName=Use Detail Dissolve})
  )
  n22 = DepthFade(Opacity=@n21, FadeDistance=ScalarParameter() {defaultValue=30, parameterName=FadeDistance})
  ShadingPathSwitch(Default=@n22, Deferred=@n22, Forward=@n21, Mobile=@n21)
 ]
}
